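# Unlit translucent material for a sweeping spotlight beam mesh, usable on instanced and Niagara meshes.
# World position offset stretches the beam along local X, widens it along normals and swings it sinusoidally about a pivot.
# Opacity combines rotation-corrected Fresnel, UV falloff, near-camera fade and time-of-day multipliers, with depth fade dropped on low quality and mobile.
# An alternative SplineThicken camera-facing setup and unused branches also remain in the graph.
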Material /Clyde/Materials/TechArt/M_Clyde_SpotlightBeam
graph {
 properties {
  bUseMaterialAttributes=true
  bUsedWithInstancedStaticMeshes=true
  bUsedWithNiagaraMeshParticles=true
  blendMode=BLEND_Translucent
  pixelDepthOffsetMode=PDOM_Legacy
  shadingModel=MSM_Unlit
  usedShadingModels=Unlit
 }
 nodes [
  n1 = ScalarParameter() {parameterName=Rotation}
  Frac(None=Max(A=@n1) {constB=0})
  n2 = Constant() {r=-0.5}
  n3 = TextureCoordinate()
  n4 = CollectionParameter() {
   collection=&FortniteMaterialParameters
   parameterName=FortniteActiveTimeOfDay
  }
  n5 = DotProduct(A=@n4, B=Constant4Vector() {constant=<rgba 0 0 0 1>}.RGBA)
  n6 = ScalarParameter() {parameterName=Time Falloff Power}
  n7 = ScalarParameter() {defaultValue=1500, parameterName=Main Width}
  n8 = StaticSwitchParameter(True=Multiply(A=Power(Base=@n5, Exp=@n6), B=@n7), False=@n7) {parameterName=Turn Off At Day}
  n9 = Multiply(A=ScalarParameter() {defaultValue=1, parameterName=Main Width End}, B=@n8)
  n10 = TextureCoordinate()
  n11 = Multiply(A=ComponentMask(None=Add(A=@n2, B=@n3)) {r=true}, B=LinearInterpolate(A=@n8, B=@n9, Alpha=ComponentMask(None=@n10) {r=true}))
  n12 = Multiply(
   A=AppendVector(A=AppendVector(A=@n11, B=@n11), B=Constant())
   B=Transform(None=VectorParameter() {parameterName=fdfd, defaultValue=<rgba 1 0 0 1>}.RGB) {transformSourceType=TRANSFORMSOURCE_Camera}
  )
  n13 = MaterialFunctionCall(WidthBase=@n8, WidthTip=@n9) {desc=Can't be done in instances, materialFunction=&SplineThicken}
  LinearInterpolate(A=@n12, B=@n13.WorldPositionOffset, Alpha=ScalarParameter() {parameterName=DebugTestImposterThicken})
  n14 = Constant3Vector() {constant=<rgba 0 0 1 0>}
  n15 = ScalarParameter() {parameterName=Rotation}
  n16 = Multiply(A=Constant() {r=34.9111}, B=Frac(None=Max(A=@n15) {constB=0})) {constA=16}
  n17 = Constant3Vector()
  n18 = VectorParameter() {parameterName=Rotation Vector, defaultValue=<rgba 1 0 0 1>}
  n19 = Multiply(
   A=Time()
   B=LinearInterpolate(
    A=ScalarParameter() {defaultValue=0.125, parameterName=Rotation Speed}
    B=ScalarParameter() {defaultValue=0.07, parameterName=Rotation Speed2}
    Alpha=Frac(None=@n16)
   )
  )
  n20 = Constant3Vector() {constant=<rgba 0 0 1 0>}
  n21 = Transform(None=Constant3Vector() {constant=<rgba 1 0 0 0>}) {transformSourceType=TRANSFORMSOURCE_Local}
  n22 = OneMinus(None=ComponentMask(None=TextureCoordinate()) {g=true})
  n23 = ScalarParameter() {defaultValue=15, parameterName=ScaleHeight}
  n24 = StaticSwitchParameter(True=Multiply(A=@n23, B=@n5), False=@n23) {parameterName=Turn Off At Day}
  n25 = Add(
   A=Multiply(A=@n13.WorldPositionOffset) {constB=-1, desc=Rotate to face camera}
   B=Multiply(A=Multiply(A=@n22, B=@n24), B=StaticSwitchParameter(True=@n20, False=@n21) {defaultValue=true, parameterName=Rotate})
  )
  n26 = RotateAboutAxis(
   NormalizedRotationAxis=Normalize(VectorInput=RotateAboutAxis(NormalizedRotationAxis=@n14, RotationAngle=@n16, PivotPoint=@n17, Position=@n18.RGB))
   RotationAngle=Multiply(A=Sine(None=@n19), B=ScalarParameter() {defaultValue=0.125, parameterName=Rotation Max})
   PivotPoint=Add(A=ActorPositionWS(), B=VectorParameter() {parameterName=Pivot Offset Custom, defaultValue=<rgba 0 0 75 1>}.RGB)
   Position=Add(A=@n25, B=WorldPosition().XYZ)
  )
  n27 = Constant()
  MakeMaterialAttributes!(
   BaseColor
   Metallic
   Specular
   Roughness
   Anisotropy
   EmissiveColor=VectorParameter() {parameterName=Color, defaultValue=<rgba 3.7163 1.2205 4 1>}.RGB
   Opacity=MaterialFunctionCall!(Width=Constant() {r=0.25}, Sharpness, Offset, Direction Switch=StaticBool() {value=true}, Compare, Input Coordinates) {materialFunction=/Engine/Functions/Engine_MaterialFunctions03/Procedurals/GeneratedBand}.Result
   OpacityMask
   Normal
   Tangent
   WorldPositionOffset=Add(A=@n25, B=StaticSwitchParameter(True=@n26, False=@n27) {defaultValue=true, parameterName=Rotate})
   SubsurfaceColor
   ClearCoat
   ClearCoatRoughness
   AmbientOcclusion
   Refraction
   CustomizedUVs_0
   CustomizedUVs_1
   CustomizedUVs_2
   CustomizedUVs_3
   CustomizedUVs_4
   CustomizedUVs_5
   CustomizedUVs_6
   CustomizedUVs_7
   PixelDepthOffset
   ShadingModel
   Displacement
  )
  n28 = Constant() {r=34.9111}
  n29 = Normalize(VectorInput=RotateAboutAxis(
    NormalizedRotationAxis=Constant3Vector() {constant=<rgba 0 0 1 0>}
    RotationAngle=Multiply(A=@n28, B=@n1) {constA=16}
    PivotPoint=Constant3Vector()
    Position=VectorParameter() {parameterName=Rotation Vector, defaultValue=<rgba 1 0 0 1>}.RGB
   ))
  n30 = Multiply(
   A=Add(A=Time(), B=ScalarParameter() {parameterName=Rotation Timing Offset})
   B=ScalarParameter() {defaultValue=0.125, parameterName=Rotation Speed}
  )
  n31 = Add(
   A=ScalarParameter() {parameterName=Rotation Angle Offset}
   B=Multiply(A=Sine(None=@n30), B=ScalarParameter() {defaultValue=0.125, parameterName=Rotation Max})
  )
  n32 = MaterialFunctionCall(
   Brightness=ScalarParameter() {defaultValue=1, parameterName=FresnelBrightness}
   Exponent=ScalarParameter() {defaultValue=1, parameterName=FresnelExp}
   Vertex / Pixel Normal=MaterialFunctionCall(Rotation Axis=@n29, Rotation Angle=@n31) {materialFunction=&FixRotateAboutAxisNormals}.New World Space Vertex Normal
  ) {materialFunction=&MF_Fresnel_FX}
  n33 = ScalarParameter() {defaultValue=1, parameterName=OpacityFalloffMult.TOD.Morning}
  n34 = ScalarParameter() {defaultValue=1, parameterName=OpacityFalloffMult.TOD.Day}
  n35 = ScalarParameter() {defaultValue=1, parameterName=OpacityFalloffMult.TOD.Evening}
  n36 = ScalarParameter() {defaultValue=1, parameterName=OpacityFalloffMult.TOD.Night}
  n37 = Power(
   Base=ComponentMask(None=TextureCoordinate()) {g=true}
   Exp=Multiply(
    A=ScalarParameter() {defaultValue=3, parameterName=Falloff}
    B=MaterialFunctionCall(Value_Morning=@n33, Value_Day=@n34, Value_Evening=@n35, Value_Night=@n36) {materialFunction=&Scalar_Time_of_Day}.Result
   )
  )
  n38 = ComponentMask(None=TextureCoordinate()) {g=true}
  n39 = ScalarParameter() {parameterName=FadeInOffset}
  n40 = Saturate(None=Power(Base=Add(A=@n38, B=@n39), Exp=ScalarParameter() {defaultValue=1, parameterName=FadeIn}))
  n41 = Multiply(A=@n40, B=@n40)
  n42 = ScalarParameter() {parameterName=FadeInAddAlpha}
  n43 = WorldPosition()
  n44 = Constant() {r=0}
  n45 = TextureSample!(UVs=ComponentMask(None=Multiply(A=@n43.XYZ, B=@n44)) {g=true, r=true}, Tex, Bias, Apply View MipBias) {
   mipValueMode=TMVM_MipBias
   samplerSource=SSM_Wrap_WorldGroupSettings
   samplerType=SAMPLERTYPE_LinearColor
   texture=/Game/Global/Textures/Default/Tileable_Patterns/Cloud_Diffuse
  }
  n46 = Multiply(A=WorldPosition().XYZ, B=Multiply(A=@n44) {constB=1.185})
  n47 = Multiply(
   A=@n45.G
   B=TextureSample!(UVs=ComponentMask(None=@n46) {b=true, r=true}, Tex, Apply View MipBias) {samplerType=SAMPLERTYPE_LinearColor, texture=&Cloud_Diffuse}.G
  )
  n48 = Multiply(A=@n47, B=DepthFade())
  n49 = QualitySwitch(Default=@n48, Low=@n47, High=@n48, Medium=@n48)
  n50 = ScalarParameter() {
   defaultValue=1
   desc=Default HDRMult is too bright for mobile. This allows to reduce for mobile when HDR is high.
   parameterName=MobileBrightnessMultiplier
  }
  n51 = Multiply(A=@n47, B=@n50)
  n52 = Multiply(
   A=Add(A=Multiply(A=@n32.Fresnel, B=@n37), B=LinearInterpolate(B=@n41, Alpha=@n42))
   B=Multiply(
    A=FeatureLevelSwitch(Default=@n49, ES2=@n51, ES3_1=@n51, SM4_REMOVED=@n49, SM5=@n49, SM6=@n49)
    B=ScalarParameter() {defaultValue=1, parameterName=OpacityBoost}
   )
  )
  n53 = MaterialFunctionCall(Distance=ScalarParameter() {defaultValue=500, parameterName=NearFade}) {materialFunction=&Near_Camera_Fade}
  n54 = ScalarParameter() {defaultValue=1, parameterName=OpacityMult.TOD.Morning}
  n55 = ScalarParameter() {defaultValue=1, parameterName=OpacityMult.TOD.Day}
  n56 = ScalarParameter() {defaultValue=1, parameterName=OpacityMult.TOD.Evening}
  n57 = ScalarParameter() {defaultValue=1, parameterName=OpacityMult.TOD.Night}
  n58 = VectorParameter() {parameterName=Pivot Offset Custom, defaultValue=<rgba 0 0 75 1>}
  n59 = Normalize(VectorInput=Transform(None=Constant3Vector() {constant=<rgba 1 0 0 0>}) {transformSourceType=TRANSFORMSOURCE_Local})
  n60 = Normalize(VectorInput=Subtract(
    A=WorldPosition().XYZ
    B=MaterialFunctionCall() {materialFunction=&ComponentPivotLocation}.Result
   ))
  n61 = SphereMask(
   A=MaterialFunctionCall() {materialFunction=&ComponentPivotLocation}.Result
   B=WorldPosition().XYZ
   Radius=ScalarParameter() {defaultValue=1.2, parameterName=WPO.CenterMaskRadius}
  )
  n62 = ScalarParameter() {defaultValue=15, parameterName=ScaleHeight}
  n63 = ScalarParameter() {parameterName=WPO.ScaleInner}
  n64 = ScalarParameter() {defaultValue=150, parameterName=MainWidth}
  n65 = Add(
   A=Multiply(A=LinearInterpolate(A=@n59, B=@n60, Alpha=@n61), B=LinearInterpolate(A=@n62, B=@n63, Alpha=@n61) {constB=0})
   B=Multiply(A=LinearInterpolate(A=@n64, Alpha=@n61) {constA=1, constB=0}, B=VertexNormalWS())
  )
  n66 = RotateAboutAxis(
   NormalizedRotationAxis=@n29
   RotationAngle=@n31
   PivotPoint=Add(A=ActorPositionWS(), B=@n58.RGB)
   Position=Add(A=@n65, B=WorldPosition().XYZ)
  )
  n67 = Constant()
  MakeMaterialAttributes(
   EmissiveColor=VectorParameter() {parameterName=Color, defaultValue=<rgba 3.7163 1.2205 4 1>}.RGB
   Opacity=Multiply(
    A=Multiply(A=@n52, B=@n53.Result)
    B=MaterialFunctionCall(Value_Morning=@n54, Value_Day=@n55, Value_Evening=@n56, Value_Night=@n57) {materialFunction=&Scalar_Time_of_Day}.Result
   )
   WorldPositionOffset=Add(A=@n65, B=StaticSwitchParameter(True=@n66, False=@n67) {defaultValue=true, parameterName=Rotate})
  )
  n68 = TextureCoordinate()
  n69 = ScalarParameter() {defaultValue=1.5, parameterName=Scale Tip}
  n70 = ScalarParameter() {defaultValue=0.2, parameterName=Scale Base}
  n71 = ComponentMask(None=@n68) {g=true}
  n72 = ComponentMask(None=MaterialFunctionCall(UVs=@n68, Texture Scale=LinearInterpolate(A=@n69, B=@n70, Alpha=@n71)) {materialFunction=&ScaleUVsByCenter}.UVs) {r=true}
  n73 = Multiply(A=@n72, B=OneMinus(None=@n72))
  n74 = ScalarParameter() {defaultValue=3, parameterName=Edge Fade Power}
  n75 = Saturate(None=Multiply(A=Power(Base=@n73, Exp=@n74) {constExponent=4}, B=ScalarParameter() {defaultValue=5, parameterName=Center Column Multiplier}))
  n76 = ComponentMask(None=TextureCoordinate()) {g=true}
  n77 = Power(Base=@n76, Exp=ScalarParameter() {defaultValue=1, parameterName=Tip Fade Power})
  n78 = Multiply(
   A=Power(Base=@n76, Exp=ScalarParameter() {defaultValue=10, parameterName=Tip Fade Power2})
   B=ScalarParameter() {defaultValue=3, parameterName=Tip Fade Power2_Brightness}
  )
  n79 = WorldPosition()
  n80 = Constant() {r=0}
  n81 = TextureSample!(UVs=ComponentMask(None=Multiply(A=@n79.XYZ, B=@n80)) {g=true, r=true}, Tex, Bias, Apply View MipBias) {
   mipValueMode=TMVM_MipBias
   samplerSource=SSM_Wrap_WorldGroupSettings
   samplerType=SAMPLERTYPE_LinearColor
   texture=&Cloud_Diffuse
  }
  n82 = ComponentMask(None=Multiply(A=WorldPosition().XYZ, B=Multiply(A=@n80) {constB=1.185})) {b=true, r=true}
  n83 = Multiply(A=Multiply(
    A=@n81.G
    B=TextureSample!(UVs=@n82, Tex, Apply View MipBias) {samplerType=SAMPLERTYPE_LinearColor, texture=&Cloud_Diffuse}.G
   ))
  n84 = ScalarParameter() {defaultValue=0.1, parameterName=Thicken Beam}
  n85 = Multiply(A=Add(A=Multiply(A=@n75, B=@n77), B=Multiply(A=@n75, B=@n78)), B=Saturate(None=Add(A=@n83, B=@n84)))
  n86 = Multiply(A=@n85, B=DepthFade())
  n87 = QualitySwitch(Default=@n86, Low=@n85, High=@n86, Medium=@n86)
  n88 = ScalarParameter() {
   defaultValue=1
   desc=Default HDRMult is too bright for mobile. This allows to reduce for mobile when HDR is high.
   parameterName=MobileBrightnessMultiplier
  }
  n89 = Multiply(A=@n85, B=@n88)
  n90 = Multiply(
   A=FeatureLevelSwitch(Default=@n87, ES2=@n89, ES3_1=@n89, SM4_REMOVED=@n87, SM5=@n87, SM6=@n87)
   B=MaterialFunctionCall(Distance=ScalarParameter() {defaultValue=500, parameterName=NearFade}) {materialFunction=&Near_Camera_Fade}.Result
  )
  Multiply(
   A=@n90
   B=MaterialFunctionCall(
    Fade_Start_Distance=ScalarParameter() {defaultValue=250000, parameterName="Fade_Start_Distance (S)"}
    Fade_End_Distance=ScalarParameter() {defaultValue=150000, parameterName="Fade_End_Distance (S)"}
   ) {materialFunction=&MF_Distance_SyncedFade_FX}.Result Inverted
  )
  OneMinus(None=@n38)
  Reroute(None=@n12)
  ScalarParameter() {parameterName=rotation}
  n91 = CollectionParameter() {
   collection=&FortniteMaterialParameters
   parameterName=FortniteActiveTimeOfDay
  }
  n92 = Constant4Vector() {constant=<rgba 0 0 0 1>}
  StaticSwitchParameter(True=LinearInterpolate(B=@n90, Alpha=DotProduct(A=@n91, B=@n92.RGBA)), False=@n90) {parameterName=Turn Off At Day}
  n93 = ScalarParameter() {parameterName=rotation}
  n94 = Multiply(A=Constant() {r=34.9111}, B=Frac(None=Max(A=@n93) {constB=0})) {constA=16}
  n95 = RotateAboutAxis(
   NormalizedRotationAxis=Constant3Vector() {constant=<rgba 0 0 1 0>}
   RotationAngle=@n94
   PivotPoint=Constant3Vector()
   Position=VectorParameter() {parameterName=Rotation Vector, defaultValue=<rgba 1 0 0 1>}.RGB
  )
  n96 = ScalarParameter() {defaultValue=0.125, parameterName=Rotation Speed}
  n97 = ScalarParameter() {defaultValue=0.07, parameterName=Rotation Speed2}
  n98 = Frac(None=@n94)
  n99 = Sine(None=Multiply(A=Time(), B=LinearInterpolate(A=@n96, B=@n97, Alpha=@n98)))
  n100 = ScalarParameter() {defaultValue=0.125, parameterName=Rotation Max}
  n101 = ActorPositionWS()
  n102 = Transform(None=VectorParameter() {parameterName=Pivot Offset Custom, defaultValue=<rgba 0 0 75 1>}.RGB) {transformSourceType=TRANSFORMSOURCE_Local}
  n103 = WorldPosition()
  StaticSwitchParameter(
   True=RotateAboutAxis(
    NormalizedRotationAxis=Normalize(VectorInput=@n95)
    RotationAngle=Multiply(A=@n99, B=@n100)
    PivotPoint=Add(A=@n101, B=@n102)
    Position=Add(A=@n103.XYZ, B=@n65)
   )
   False=Constant()
  ) {defaultValue=true, parameterName=Rotate}
  Transform(None=@n58.RGB) {transformSourceType=TRANSFORMSOURCE_Local}
 ]
}
